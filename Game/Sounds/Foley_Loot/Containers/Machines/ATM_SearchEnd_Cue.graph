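# Sound cue for the end of searching an ATM loot container, classed as a loot collect foley sound.
# A single wave player feeds an Enveloper that fades the ATM wave from full volume to silence over 0.3 seconds, leaving a short tail.
# Inline attenuation overrides take the place of the assigned prop-search attenuation asset, and concurrency overrides block new instances once the limit is reached.
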SoundCue /Game/Sounds/Foley_Loot/Containers/Machines/ATM_SearchEnd_Cue
root=Enveloper_3
nodes [
 {
  id=Enveloper_3
  children [WavePlayer_0]
  properties {pitchCurve {editorCurveData {keys [{time=0}]}}, volumeCurve {editorCurveData {keys [<at 0 1>,<at 0.3 0>]}}}
 }
 {id=WavePlayer_0, properties {soundWaveAssetPtr=&/ATM}}
]
properties {
 attenuationSettings=/Game/Sounds/Attenuation/Fort_Prop_Search_Att
 bOverrideAttenuation=true
 bOverrideConcurrency=true
 maxDistance=1500
 soundClassObject=&Fort_Loot_Collect
 volumeMultiplier=0.7
 attenuationOverrides {falloffDistance=1000, attenuationShapeExtents {x=500}}
 concurrencyOverrides {resolutionRule=PreventNew}
}
