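# Sound cue for searching an ATM loot container, classed as loot-collect foley.
# A quality-level switch selects between two modulated layered mixes and, for its last two levels, a single optimized ATM wave.
# The first mix layers the ATM search sound with a weighted random gear rustle, while the second uses one fixed gear sound.
# Prop-search attenuation is overridden, as is concurrency, which prevents new instances at the limit.
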SoundCue /Game/Sounds/Foley_Loot/Containers/Machines/ATM_Search_Cue
root=QualityLevel_0
nodes [
 {id=Mixer_0, children [WavePlayer_2,WavePlayer_1], properties {inputVolume [1,1]}}
 {id=Mixer_2, children [WavePlayer_29, Random_7], properties {inputVolume [1,1]}}
 {id=Modulator_0, children [Mixer_0]}
 {id=Modulator_2, children [Mixer_2]}
 {id=QualityLevel_0, children [Modulator_2,Modulator_0,WavePlayer_0,WavePlayer_3]}
 {id=Random_7, children [WavePlayer_26, WavePlayer_27, WavePlayer_28], properties {weights [1,1,0.5]}}
 {id=WavePlayer_0, properties {soundWaveAssetPtr=&/ATM_Search_opt}}
 {id=WavePlayer_1, properties {soundWaveAssetPtr=/Game/Sounds/Foley_Loot/Containers/Misc/Prop_Gear_Search_04}}
 {id=WavePlayer_2, properties {soundWaveAssetPtr=&/Search_ATM_New_01}}
 {id=WavePlayer_26, properties {soundWaveAssetPtr=&Prop_Gear_Search_04}}
 {id=WavePlayer_27, properties {soundWaveAssetPtr=/Game/Sounds/Foley_Loot/Containers/Misc/Prop_Gear_Search_06}}
 {id=WavePlayer_28, properties {soundWaveAssetPtr=/Game/Sounds/Foley_Loot/Containers/Misc/Prop_Gear_Search_01}}
 {id=WavePlayer_29, properties {soundWaveAssetPtr=&Search_ATM_New_01}}
 {id=WavePlayer_3, properties {soundWaveAssetPtr=&ATM_Search_opt}}
]
properties {
 attenuationSettings=/Game/Sounds/Attenuation/Fort_Prop_Search_Att
 bOverrideAttenuation=true
 bOverrideConcurrency=true
 maxDistance=1500
 soundClassObject=&Fort_Loot_Collect
 volumeMultiplier=0.65
 attenuationOverrides {falloffDistance=1000, attenuationShapeExtents {x=500}}
 concurrencyOverrides {resolutionRule=PreventNew}
}
